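# Material function for the Skirmish backpack that computes flipbook UVs and a frame index driven by an opening-sequence curve from a curve atlas.
# The curve's alpha at time zero sets the loop period for modulo time, optionally folded into a triangle wave by the PingPongTime switch.
# Sampled red is the frame index and green the frame count, whose square root sizes the square grid that offsets and scales UV channel 1.
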MaterialFunction /Game/Accessories/FORT_Backpacks/Backpack_F_MED_Skirmish/Materials/MF_Skirmish_FlipbookCurve
graph {nodes [
  n1 = CurveAtlasRowParameter(CurveTime=Constant()) {
   atlas=&/CA_Skirmish_FlipbookCurve
   curve=&/C_Flipbook_OpeningSequence
   parameterName=CurveAtlas
  }
  n2 = MaterialFunctionCall!(Time, Divisor=@n1.A, Mod) {materialFunction=/Game/Effects/Fort_Effects/Materials/_Functions/MF_ModuloTime}
  n3 = Sign(None=ConstantBiasScale(None=@n2.Unclamped) {bias=-0.5, scale=2})
  n4 = Constant()
  n5 = Multiply(A=OneMinus(None=@n2.Unclamped)) {constB=2}
  n6 = Multiply(A=@n2.Unclamped) {constB=2}
  n7 = CurveAtlasRowParameter(CurveTime=StaticSwitchParameter(True=If(A=@n3, B=@n4, "A > B"=@n5, "A < B"=@n6), False=@n2.Unclamped) {parameterName=PingPongTime}) {
   atlas=&CA_Skirmish_FlipbookCurve
   curve=&C_Flipbook_OpeningSequence
   parameterName=CurveAtlas
  }
  n8 = SquareRoot(None=@n7.G)
  n9 = Divide(A=@n7.R, B=@n8)
  n10 = Truncate!(None=Fmod(A=AppendVector(A=@n7.R, B=@n9), B=@n8)) {is=class MaterialExpressionTruncate}
  n11 = TextureCoordinate() {coordinateIndex=1}
  FunctionOutput(None=Add(A=Divide(A=@n10, B=@n8), B=Divide(A=@n11, B=@n8))) {outputName=UVs}
  FunctionOutput(None=@n7.R) {outputName=Index, sortPriority=1}
 ]}
