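# Screen material for the Skirmish backpack accessory that plays an animated sub-UV flipbook as emissive on a dark, semi-metallic surface.
# MF_Skirmish_FlipbookCurve drives the sheet UVs, a static switch can overlay the frame index for debugging, and Fresnel modulates the glow.
# A procedural domed normal comes from centered UVs and a sphere mask, and attributes pass through TechArt overrides.
# An inline curve-atlas flipbook chain remains unconnected.
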Material /Game/Accessories/FORT_Backpacks/Backpack_F_MED_Skirmish/Materials/M_Skirmish_Screen
graph {
 properties {
  bUseMaterialAttributes=true
  bUsedWithMeshDeformer=true
  bUsedWithSkeletalMesh=true
  pixelDepthOffsetMode=PDOM_Legacy
  usedShadingModels=Default Lit
 }
 nodes [
  n1 = MaterialFunctionCall() {materialFunction=&/MF_Skirmish_FlipbookCurve}
  n2 = TextureSampleParameter2D(UVs=@n1.UVs) {parameterName=UVSheet, texture=&/T_Skirmish_SubUV_01}
  n3 = ScalarParameter() {defaultValue=2, group=DebugScalarValues, parameterName=MaximumNumberOfDigits}
  n4 = StaticSwitchParameter(
   True=Add(
    A=@n2.RGB
    B=MaterialFunctionCall(Number=@n1.Index, MaximumNumberOfDigits=@n3) {materialFunction=&DebugScalarValues}.Result
   )
   False=@n2.RGB
  ) {group=DebugScalarValues, parameterName=DebugFrameIndexScalar}
  n5 = MaterialFunctionCall(Exponent=ScalarParameter() {defaultValue=0.8, parameterName=FresnelExponent}) {materialFunction=&MF_Fresnel_FX}
  n6 = TextureCoordinate()
  n7 = MaterialFunctionCall(Float2=@n6) {materialFunction=&BreakOutFloat2Components}
  n8 = Constant() {r=0.5}
  n9 = Constant() {r=0.75}
  n10 = Constant()
  n11 = MaterialFunctionCall(
   R=Subtract(A=@n7.R) {constB=0.5}
   G=Subtract(A=@n7.G) {constB=0.5}
   B=ConstantBiasScale(None=SphereMask(A=@n6, B=@n8, Radius=@n9, Hardness=@n10))
  ) {materialFunction=&AppendMany}
  n12 = SetMaterialAttributes!(
   MaterialAttributes
   Base Color=NamedRerouteDeclaration(None=VectorParameter() {parameterName=BaseColor, defaultValue=<rgba 0.03 0.03 0.03 0>}.RGB) {name=basecolor, nodeColor=<rgba 0.8047 0 1 1>}
   Metallic=NamedRerouteDeclaration(None=ScalarParameter() {defaultValue=0.35, parameterName=Metallic}) {name=metallic, nodeColor=<rgba 1 0 0.6328 1>}
   Specular=NamedRerouteDeclaration(None=ScalarParameter() {defaultValue=0.4, parameterName=Specular}) {name=specular, nodeColor=<rgba 1 0.8906 0 1>}
   Roughness=NamedRerouteDeclaration(None=ScalarParameter() {defaultValue=0.56, parameterName=Roughness}) {name=roughness, nodeColor=<rgba 0 1 0.7188 1>}
   Emissive Color=NamedRerouteDeclaration(None=Multiply(A=@n4, B=@n5.Fresnel)) {name=emissive, nodeColor=<rgba 0.9453 0 1 1>}
   Normal=NamedRerouteDeclaration(None=Normalize(VectorInput=@n11.RGB)) {name=normal, nodeColor=<rgba 1 0 0.6797 1>}
  ) {attributeSetTypes [
    #1
    #2
    #3
    #4
    #5
    #6
   ]}
  MaterialFunctionCall(In=@n12) {materialFunction=&MF_TechArtOverrides_accessories}
  n13 = CurveAtlasRowParameter(CurveTime=Constant()) {
   atlas=&/CA_Skirmish_FlipbookCurve
   curve=&/C_Flipbook_01
   parameterName=CurveAtlas
  }
  n14 = CurveAtlasRowParameter(CurveTime=MaterialFunctionCall!(Time, Divisor=@n13.G, Mod) {materialFunction=/Game/Effects/Fort_Effects/Materials/_Functions/MF_ModuloTime}.Unclamped) {
   atlas=&CA_Skirmish_FlipbookCurve
   curve=&C_Flipbook_01
   parameterName=CurveAtlas
  }
  n15 = Divide(A=Floor(None=Multiply(A=@n14.R, B=@n13.B)), B=@n13.B)
  n16 = Divide(A=Floor(None=Multiply(A=@n13.B, B=@n13.B)), B=@n13.B)
  n17 = TextureCoordinate()
  TextureSample(UVs=Add(A=AppendVector(A=@n15, B=@n16), B=Divide(A=@n17, B=@n13.B)))
 ]
}
