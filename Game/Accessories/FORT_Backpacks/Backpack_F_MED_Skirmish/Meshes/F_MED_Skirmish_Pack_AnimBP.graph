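# Animation Blueprint for the Skirmish backpack that animates its buttons and joystick in response to the wearing player's movement and actions.
# A state machine switches from Idle to button animations for 2D acceleration, recent firing, skydiving, jumping/falling and melee, returning when each condition clears.
# Each update, PlayerMovement output builds JoystickRotator, which additively rotates the forward/back and left/right joystick bones in bone space only while the player accelerates in 2D.
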AnimBlueprint /Game/Accessories/FORT_Backpacks/Backpack_F_MED_Skirmish/Meshes/F_MED_Skirmish_Pack_AnimBP
target_skeleton=&/F_MED_Skirmish_Pack_Skeleton
graphs [
 AnimGraph {nodes [
   AnimGraphNode_StateMachine_0 = Buttons_StateMachine() {is=title AnimGraphNode_StateMachine}
   n1 = "To Float (Boolean)"(InBool=Get bIsAccelerating2D())
   n2 = Get JoystickRotator() -> (JoystickRotator:Rotator)
   n3 = "Transform (Modify) Bone - Bone: dyn_joystick_forwardBack_jnt"!(ComponentPose=Local To Component(LocalPose=@AnimGraphNode_StateMachine_0), Alpha=@n1, Translation="0,0,0", Rotation=@n2) -> (Pose) {
    rotationMode=BMM_Additive
    rotationSpace=BCS_BoneSpace
    alphaBoolBlend {blendOption=Linear}
    boneToModify {boneName=dyn_joystick_forwardBack_jnt}
   }
   n4 = "Transform (Modify) Bone - Bone: dyn_joystick_leftRight_jnt"!(ComponentPose=@n3, Alpha=@n1, Translation="0,0,0", Rotation=@n2) -> (Pose) {
    rotationMode=BMM_Additive
    rotationSpace=BCS_BoneSpace
    alphaBoolBlend {blendOption=Linear}
    boneToModify {boneName=dyn_joystick_leftRight_jnt}
   }
   Output Pose(Result=Component To Local(ComponentPose=@n4)) {layerGroup=None}
  ]}
 Buttons_StateMachine {
  entry=Idle
  path=AnimGraph.AnimGraphNode_StateMachine_0.Buttons_StateMachine
  ids=id
  nodes [
   AnimStateNode_0 = Idle()
   AnimStateNode_2 = 2D Acceleration() {is=title AnimStateNode}
   AnimStateNode_3 = Firing/RecentlyFired() {is=title AnimStateNode}
   AnimStateNode_4 = Skydiving() {is=title AnimStateNode}
   AnimStateNode_5 = Jumping() {is=title AnimStateNode}
   AnimStateNode_6 = MeleeAnim() {is=title AnimStateNode}
   AnimStateTransitionNode_1 = AnimStateTransitionNode()
   AnimStateTransitionNode_10 = AnimStateTransitionNode()
   AnimStateTransitionNode_2 = AnimStateTransitionNode()
   AnimStateTransitionNode_3 = AnimStateTransitionNode()
   AnimStateTransitionNode_4 = AnimStateTransitionNode()
   AnimStateTransitionNode_5 = AnimStateTransitionNode()
   AnimStateTransitionNode_6 = AnimStateTransitionNode()
   AnimStateTransitionNode_7 = AnimStateTransitionNode()
   AnimStateTransitionNode_8 = AnimStateTransitionNode()
   AnimStateTransitionNode_9 = AnimStateTransitionNode()
  ]
  links [
   {
    dst=Idle
    rule=AnimGraph.AnimGraphNode_StateMachine_0.Buttons_StateMachine.AnimStateTransitionNode_3.Transition
    src=2D Acceleration
    via=AnimStateTransitionNode_3
   }
   {
    dst=Idle
    rule=AnimGraph.AnimGraphNode_StateMachine_0.Buttons_StateMachine.AnimStateTransitionNode_4.Transition
    src=Firing/RecentlyFired
    via=AnimStateTransitionNode_4
   }
   {
    dst=2D Acceleration
    rule=AnimGraph.AnimGraphNode_StateMachine_0.Buttons_StateMachine.AnimStateTransitionNode_10.AnimationTransitionGraph_0
    src=Idle
    via=AnimStateTransitionNode_10
   }
   {
    dst=Firing/RecentlyFired
    rule=AnimGraph.AnimGraphNode_StateMachine_0.Buttons_StateMachine.AnimStateTransitionNode_1.Transition
    src=Idle
    via=AnimStateTransitionNode_1
   }
   {
    dst=Jumping
    rule=AnimGraph.AnimGraphNode_StateMachine_0.Buttons_StateMachine.AnimStateTransitionNode_6.Transition
    src=Idle
    via=AnimStateTransitionNode_6
   }
   {
    dst=MeleeAnim
    rule=AnimGraph.AnimGraphNode_StateMachine_0.Buttons_StateMachine.AnimStateTransitionNode_7.Transition
    src=Idle
    via=AnimStateTransitionNode_7
   }
   {
    dst=Skydiving
    rule=AnimGraph.AnimGraphNode_StateMachine_0.Buttons_StateMachine.AnimStateTransitionNode_2.Transition
    src=Idle
    via=AnimStateTransitionNode_2
   }
   {
    dst=Idle
    rule=AnimGraph.AnimGraphNode_StateMachine_0.Buttons_StateMachine.AnimStateTransitionNode_8.Transition
    src=Jumping
    via=AnimStateTransitionNode_8
   }
   {
    dst=Idle
    rule=AnimGraph.AnimGraphNode_StateMachine_0.Buttons_StateMachine.AnimStateTransitionNode_9.Transition
    src=MeleeAnim
    via=AnimStateTransitionNode_9
   }
   {
    dst=Idle
    rule=AnimGraph.AnimGraphNode_StateMachine_0.Buttons_StateMachine.AnimStateTransitionNode_5.Transition
    src=Skydiving
    via=AnimStateTransitionNode_5
   }
  ]
 }
 Idle {
  path=AnimGraph.AnimGraphNode_StateMachine_0.Buttons_StateMachine.AnimStateNode_0.Idle
  nodes [
   Sequence Player 'Skirmish_Pack_Button1'() {bLoopAnimation=false, sequence=/Game/Accessories/FORT_Backpacks/Backpack_F_MED_Skirmish/Animations/Sequence/Skirmish_Pack_Button1}
   Output Animation Pose()
  ]
 }
 2D Acceleration {
  path=AnimGraph.AnimGraphNode_StateMachine_0.Buttons_StateMachine.AnimStateNode_2.2D Acceleration
  nodes [Output Animation Pose(Result=Blend Poses by bool(
     BlendPose_0=Sequence Player 'Skirmish_Pack_Button2'() {sequence=/Game/Accessories/FORT_Backpacks/Backpack_F_MED_Skirmish/Animations/Sequence/Skirmish_Pack_Button2}
     BlendPose_1=Sequence Player 'Skirmish_Pack_Button1'() {sequence=&Skirmish_Pack_Button1}
     bActiveValue=Get bIsAccelerating2D()
    ) {blendType=Linear})]
 }
 Firing/RecentlyFired {
  path=AnimGraph.AnimGraphNode_StateMachine_0.Buttons_StateMachine.AnimStateNode_3.Firing/RecentlyFired
  nodes [Output Animation Pose(Result=Blend Poses by bool(
     BlendPose_0=Sequence Player 'Skirmish_Pack_Button3'() {sequence=/Game/Accessories/FORT_Backpacks/Backpack_F_MED_Skirmish/Animations/Sequence/Skirmish_Pack_Button3}
     BlendPose_1=Sequence Player 'Skirmish_Pack_Button1'() {sequence=&Skirmish_Pack_Button1}
     bActiveValue=Get bRecentlyFired() -> (bRecentlyFired:bool)
    ) {blendType=Linear})]
 }
 Skydiving {
  path=AnimGraph.AnimGraphNode_StateMachine_0.Buttons_StateMachine.AnimStateNode_4.Skydiving
  nodes [
   n1 = Get bIsSkyDiving()
   n2 = Get bIsSkydiveFloating()
   Output Animation Pose(Result=Blend Poses by bool(
     BlendPose_0=Sequence Player 'Skirmish_Pack_Button4'() {sequence=/Game/Accessories/FORT_Backpacks/Backpack_F_MED_Skirmish/Animations/Sequence/Skirmish_Pack_Button4}
     BlendPose_1=Sequence Player 'Skirmish_Pack_Button1'() {sequence=&Skirmish_Pack_Button1}
     bActiveValue=OR Boolean(A=@n2, B=@n1)
    ) {blendType=Linear})
  ]
 }
 Jumping {
  path=AnimGraph.AnimGraphNode_StateMachine_0.Buttons_StateMachine.AnimStateNode_5.Jumping
  nodes [
   n1 = Get bIsFalling()
   n2 = Get bIsJumping()
   Output Animation Pose(Result=Blend Poses by bool(
     BlendPose_0=Sequence Player 'Skirmish_Pack_Button5'() {sequence=/Game/Accessories/FORT_Backpacks/Backpack_F_MED_Skirmish/Animations/Sequence/Skirmish_Pack_Button5}
     BlendPose_1=Sequence Player 'Skirmish_Pack_Button1'() {sequence=&Skirmish_Pack_Button1}
     bActiveValue=OR Boolean(A=@n2, B=@n1)
    ) {blendType=Linear})
  ]
 }
 MeleeAnim {
  path=AnimGraph.AnimGraphNode_StateMachine_0.Buttons_StateMachine.AnimStateNode_6.MeleeAnim
  nodes [Output Animation Pose(Result=Blend Poses by bool(
     BlendPose_0=Sequence Player 'Skirmish_Pack_Button6'() {sequence=/Game/Accessories/FORT_Backpacks/Backpack_F_MED_Skirmish/Animations/Sequence/Skirmish_Pack_Button6}
     BlendPose_1=Sequence Player 'Skirmish_Pack_Button1'() {sequence=&Skirmish_Pack_Button1}
     bActiveValue=Get bIsPlayingMeleeAnim()
    ) {blendType=Linear})]
 }
 Transition {
  path=AnimGraph.AnimGraphNode_StateMachine_0.Buttons_StateMachine.AnimStateTransitionNode_1.Transition
  nodes [Result(bCanEnterTransition=Get bRecentlyFired())]
 }
 AnimationTransitionGraph_0 {
  path=AnimGraph.AnimGraphNode_StateMachine_0.Buttons_StateMachine.AnimStateTransitionNode_10.AnimationTransitionGraph_0
  nodes [Result(bCanEnterTransition=Get bIsAccelerating2D())]
 }
 Transition {
  path=AnimGraph.AnimGraphNode_StateMachine_0.Buttons_StateMachine.AnimStateTransitionNode_2.Transition
  nodes [Result(bCanEnterTransition=OR Boolean(A=Get bIsSkydiveFloating(), B=Get bIsSkyDiving()))]
 }
 Transition {
  path=AnimGraph.AnimGraphNode_StateMachine_0.Buttons_StateMachine.AnimStateTransitionNode_3.Transition
  nodes [Result(bCanEnterTransition=NOT Boolean(A=Get bIsAccelerating2D()))]
 }
 Transition {
  path=AnimGraph.AnimGraphNode_StateMachine_0.Buttons_StateMachine.AnimStateTransitionNode_4.Transition
  nodes [Result(bCanEnterTransition=NOT Boolean(A=Get bRecentlyFired()))]
 }
 Transition {
  path=AnimGraph.AnimGraphNode_StateMachine_0.Buttons_StateMachine.AnimStateTransitionNode_5.Transition
  nodes [Result(bCanEnterTransition=NOR Boolean!(A:bool=Get bIsSkydiveFloating(), B:bool=Get bIsSkyDiving()) -> (ReturnValue:bool))]
 }
 Transition {
  path=AnimGraph.AnimGraphNode_StateMachine_0.Buttons_StateMachine.AnimStateTransitionNode_6.Transition
  nodes [Result(bCanEnterTransition=OR Boolean(A=Get bIsJumping(), B=Get bIsFalling()))]
 }
 Transition {
  path=AnimGraph.AnimGraphNode_StateMachine_0.Buttons_StateMachine.AnimStateTransitionNode_7.Transition
  nodes [Result(bCanEnterTransition=Get bIsPlayingMeleeAnim())]
 }
 Transition {
  path=AnimGraph.AnimGraphNode_StateMachine_0.Buttons_StateMachine.AnimStateTransitionNode_8.Transition
  nodes [Result(bCanEnterTransition=NOR Boolean(A=Get bIsJumping(), B=Get bIsFalling()))]
 }
 Transition {
  path=AnimGraph.AnimGraphNode_StateMachine_0.Buttons_StateMachine.AnimStateTransitionNode_9.Transition
  nodes [Result(bCanEnterTransition=NOT Boolean(A=Get bIsPlayingMeleeAnim()))]
 }
 EventGraph {nodes [
   Event BlueprintInitializeAnimation()
   > Set SkelMesh(SkelMesh:Skeletal Mesh Object Reference=&/F_MED_Skirmish_Pack) -> (Output_Get:Skeletal Mesh Object Reference)
   > n1 = Cast To FortPlayerPawn(Object=TryGetFortPawn())
   > Set Pawn(Pawn:Fort Player Pawn Object Reference=@n1) -> (Output_Get:Fort Player Pawn Object Reference)
   Event BlueprintUpdateAnimation()
   > Get!(execute:Exec) -> (then:Exec, else:Exec, SkelMesh:Skeletal Mesh Object Reference)
   > Get!(execute) -> (then, else, Pawn:Fort Player Pawn Object Reference)
   > Sequence()
   > n2 = PlayerMovement(AnimInstance=Self-Reference(), XDotProductCoefficient=25, YDotProductCoefficient=25, ZDotProductCoefficient=1)
   > Set JoystickRotator(JoystickRotator:Rotator=MakeRotator(Pitch=@n2, Yaw=@n2.YMovement)) -> (Output_Get:Rotator)
  ]}
]
variables [
 BackwardDirection {type=Vector}
 ForwardDirection {type=Vector}
 JoystickRotator {type=Rotator}
 JoystickRotator_LeftRight {type=Rotator}
 LeftDirection {type=Vector}
 Pawn {type=FortPlayerPawn}
 RightDirection {type=Vector}
 SkelMesh {type=SkeletalMesh, default=&F_MED_Skirmish_Pack}
]
